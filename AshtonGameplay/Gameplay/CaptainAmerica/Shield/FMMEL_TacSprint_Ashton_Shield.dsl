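; Movement extension Blueprint named for tactical sprint with Captain America's shield in the Ashton gameplay content.
; It implements the extension activate and deactivate events, each of which receives a runtime context by reference.
; Both events only print a debug string, "activate" or "deactivate", so the graph holds no sprint or shield logic and appears to be a stub.
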(event EventOnExtensionActivate
  (params (RuntimeContext :type MMERuntime%20Context%20Structure%20%28by%20ref%29))
  (Development|PrintString 0 "activate"))

(event EventOnExtensionDeactivate
  (params (RuntimeContext :type MMERuntime%20Context%20Structure%20%28by%20ref%29))
  (Development|PrintString 0 "deactivate"))
